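# Conversation database for a Season 33 main-questline NPC story chat, in which the speaker warns of a dark presence in the Spirit Realm holding the Shard.
# One entry tag leads into a linear chain of four Speech nodes advanced by player choices.
# The opening choice requires the Story Chat Q01 quest to be active and carries the main-questline category tag and icon.
# The final Continue choice carries the chat's quest tag.
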ConversationDatabase /BattlePassS33_Quests/Conversations/StoryDialogue/StopMM/Conversation_S33_Chat_Q01
entry_points [{tag=Conversation.Entry.Season33.Story.Chat.01, destinations [#1]}]
nodes [
 ConversationEntryPointNode_0 {
  id=#1
  class=ConversationEntryPointNode
  outputs [#2]
  properties {entryTag {tagName=Conversation.Entry.Season33.Story.Chat.01}}
 }
 Speech_C_3 {
  id=#2
  class=Speech_C
  outputs [#3]
  properties {
   softTaskIcon=&T_UI_NPC_MainQuest
   generalConfig {defaultMessage="In the Spirit Realm... the demons tried to break me. They failed, but..."}
  }
  sub_nodes [
   ConversationChoiceNode_5 {
    class=ConversationChoiceNode
    properties {defaultChoiceDisplayText="Jade && the Comet", choiceTags {gameplayTags [{tagName=Conversation.Choice.Category.MainQuestline}]}}
   }
   FortConversationRequirement_ControllerRequirement_4 {
    class=FortConversationRequirement_ControllerRequirement
    properties {
     nodeName=Require: Is Quest Active
     requirement {
      class=FortControllerRequirement_IsQuestActive
      properties {questDefinition=/BattlePassS33_Quests/Items/QuestItems/Story/S33_Story_Chat/quest_S33_story_chat_q01}
     }
    }
   }
  ]
 }
 Speech_C_28 {
  id=#3
  class=Speech_C
  outputs [#4]
  properties {generalConfig {defaultMessage=There's a dark presence there. It has a plan... And it has the Shard.}}
  sub_nodes [ConversationChoiceNode_6 {class=ConversationChoiceNode, properties {defaultChoiceDisplayText=Continue}}]
 }
 Speech_C_0 {
  id=#4
  class=Speech_C
  outputs [#5]
  properties {generalConfig {defaultMessage=My brothers... neither of them are strong enough to face what lies in the Spirit Realm.}}
  sub_nodes [ConversationChoiceNode_7 {class=ConversationChoiceNode, properties {defaultChoiceDisplayText=Continue}}]
 }
 Speech_C_1 {
  id=#5
  class=Speech_C
  outputs []
  properties {generalConfig {defaultMessage=Kendo knows better than to try. It's Daigo that I'm worried about.}}
  sub_nodes [ConversationChoiceNode_8 {
    class=ConversationChoiceNode
    properties {defaultChoiceDisplayText=Continue, choiceTags {gameplayTags [{tagName=Athena.Quests.S33.Conversation.Story.Chat.01}]}}
   }]
 }
]
