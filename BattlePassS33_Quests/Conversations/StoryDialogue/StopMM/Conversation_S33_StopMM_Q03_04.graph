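# Season 33 "StopMM" story conversation 03_04, a single line of NPC dialogue about tracing the comet's energy further up the mountain.
# One entry point, keyed by its story tag, leads straight to a terminal Speech node with an intel icon.
# The speech offers a tagged "Look into the Portal" choice and requires the Story Ally Q03 quest to be active.
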ConversationDatabase /BattlePassS33_Quests/Conversations/StoryDialogue/StopMM/Conversation_S33_StopMM_Q03_04
entry_points [{tag=Conversation.Entry.Season33.Story.StopMM.03.04, destinations [#1]}]
nodes [
 ConversationEntryPointNode_3 {
  id=#1
  class=ConversationEntryPointNode
  outputs [#2]
  properties {entryTag {tagName=Conversation.Entry.Season33.Story.StopMM.03.04}}
 }
 Speech_C_7 {
  id=#2
  class=Speech_C
  outputs []
  properties {
   softTaskIcon=&T_UI_NPC_Intel2
   generalConfig {defaultMessage=The traces of the comet's energy are stronger here. Could it be further up the mountain?}
  }
  sub_nodes [
   ConversationChoiceNode_7 {
    class=ConversationChoiceNode
    properties {defaultChoiceDisplayText=Look into the Portal, choiceTags {gameplayTags [{tagName=Athena.Quests.S33.Conversation.Story.StopMM.03.04}]}}
   }
   FortConversationRequirement_ControllerRequirement_7 {
    class=FortConversationRequirement_ControllerRequirement
    properties {
     nodeName=Require: Is Quest Active
     requirement {
      class=FortControllerRequirement_IsQuestActive
      properties {questDefinition=/BattlePassS33_Quests/Items/QuestItems/Story/S33_Story_Ally/quest_S33_story_ally_q03}
     }
    }
   }
  ]
 }
]
