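# Story dialogue conversation for Season 33 StopMM quest step 03_06, in which a speaker reports weak traces of the comet's energy and says it must be further on.
# An entry point tag leads to a single speech node with an intel icon and no further outputs.
# The speech carries a "Look into the Portal" choice tagged for the quest, and a requirement gates it on the story ally Q03 quest being active.
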ConversationDatabase /BattlePassS33_Quests/Conversations/StoryDialogue/StopMM/Conversation_S33_StopMM_Q03_06
entry_points [{tag=Conversation.Entry.Season33.Story.StopMM.03.06, destinations [#1]}]
nodes [
 ConversationEntryPointNode_5 {
  id=#1
  class=ConversationEntryPointNode
  outputs [#2]
  properties {entryTag {tagName=Conversation.Entry.Season33.Story.StopMM.03.06}}
 }
 Speech_C_5 {
  id=#2
  class=Speech_C
  outputs []
  properties {
   softTaskIcon=&T_UI_NPC_Intel2
   generalConfig {defaultMessage="Traces of the comet's energy, but they're weak. It must be further on."}
  }
  sub_nodes [
   ConversationChoiceNode_4 {
    class=ConversationChoiceNode
    properties {defaultChoiceDisplayText=Look into the Portal, choiceTags {gameplayTags [{tagName=Athena.Quests.S33.Conversation.Story.StopMM.03.06}]}}
   }
   FortConversationRequirement_ControllerRequirement_5 {
    class=FortConversationRequirement_ControllerRequirement
    properties {
     nodeName=Require: Is Quest Active
     requirement {
      class=FortControllerRequirement_IsQuestActive
      properties {questDefinition=/BattlePassS33_Quests/Items/QuestItems/Story/S33_Story_Ally/quest_S33_story_ally_q03}
     }
    }
   }
  ]
 }
]
