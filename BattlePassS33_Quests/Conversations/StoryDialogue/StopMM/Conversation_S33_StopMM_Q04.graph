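# Conversation database for the Season 33 StopMM story quest 04, supplying short dialogue lines about boosting traces of the comet's energy.
# Five tagged entry points each lead to a single terminal Speech node with its own message.
# Each speech carries a "Look into the Portal" choice with a matching per-step tag and only qualifies while quest_S33_story_ally_q04 is active.
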ConversationDatabase /BattlePassS33_Quests/Conversations/StoryDialogue/StopMM/Conversation_S33_StopMM_Q04
entry_points [
 {tag=Conversation.Entry.Season33.Story.StopMM.04.01, destinations [#1]}
 {tag=Conversation.Entry.Season33.Story.StopMM.04.02, destinations [#2]}
 {tag=Conversation.Entry.Season33.Story.StopMM.04.03, destinations [#3]}
 {tag=Conversation.Entry.Season33.Story.StopMM.04.04, destinations [#4]}
 {tag=Conversation.Entry.Season33.Story.StopMM.04.05, destinations [#5]}
]
nodes [
 ConversationEntryPointNode_10 {
  id=#5
  class=ConversationEntryPointNode
  outputs [#6]
  properties {entryTag {tagName=Conversation.Entry.Season33.Story.StopMM.04.05}}
 }
 Speech_C_11 {
  id=#6
  class=Speech_C
  outputs []
  properties {softTaskIcon=&T_UI_NPC_Intel2, generalConfig {defaultMessage="Increasing spirit guidance!"}}
  sub_nodes [
   ConversationChoiceNode_11 {
    class=ConversationChoiceNode
    properties {defaultChoiceDisplayText=Look into the Portal, choiceTags {gameplayTags [{tagName=Athena.Quests.S33.Conversation.Story.StopMM.04.05}]}}
   }
   FortConversationRequirement_ControllerRequirement_11 {
    class=FortConversationRequirement_ControllerRequirement
    properties {
     nodeName=Require: Is Quest Active
     requirement {
      class=FortControllerRequirement_IsQuestActive
      properties {questDefinition=/BattlePassS33_Quests/Items/QuestItems/Story/S33_Story_Ally/quest_S33_story_ally_q04}
     }
    }
   }
  ]
 }
 ConversationEntryPointNode_9 {
  id=#4
  class=ConversationEntryPointNode
  outputs [#7]
  properties {entryTag {tagName=Conversation.Entry.Season33.Story.StopMM.04.04}}
 }
 Speech_C_9 {
  id=#7
  class=Speech_C
  outputs []
  properties {softTaskIcon=&T_UI_NPC_Intel2, generalConfig {defaultMessage=Empowering energy traces of the comet...}}
  sub_nodes [
   ConversationChoiceNode_9 {
    class=ConversationChoiceNode
    properties {defaultChoiceDisplayText=Look into the Portal, choiceTags {gameplayTags [{tagName=Athena.Quests.S33.Conversation.Story.StopMM.04.04}]}}
   }
   FortConversationRequirement_ControllerRequirement_9 {
    class=FortConversationRequirement_ControllerRequirement
    properties {
     nodeName=Require: Is Quest Active
     requirement {
      class=FortControllerRequirement_IsQuestActive
      properties {questDefinition=&quest_S33_story_ally_q04}
     }
    }
   }
  ]
 }
 ConversationEntryPointNode_8 {
  id=#3
  class=ConversationEntryPointNode
  outputs [#8]
  properties {entryTag {tagName=Conversation.Entry.Season33.Story.StopMM.04.03}}
 }
 Speech_C_8 {
  id=#8
  class=Speech_C
  outputs []
  properties {softTaskIcon=&T_UI_NPC_Intel2, generalConfig {defaultMessage=Boosting traces of the comet's energy...}}
  sub_nodes [
   ConversationChoiceNode_8 {
    class=ConversationChoiceNode
    properties {defaultChoiceDisplayText=Look into the Portal, choiceTags {gameplayTags [{tagName=Athena.Quests.S33.Conversation.Story.StopMM.04.03}]}}
   }
   FortConversationRequirement_ControllerRequirement_8 {
    class=FortConversationRequirement_ControllerRequirement
    properties {
     nodeName=Require: Is Quest Active
     requirement {
      class=FortControllerRequirement_IsQuestActive
      properties {questDefinition=&quest_S33_story_ally_q04}
     }
    }
   }
  ]
 }
 ConversationEntryPointNode_7 {
  id=#2
  class=ConversationEntryPointNode
  outputs [#9]
  properties {entryTag {tagName=Conversation.Entry.Season33.Story.StopMM.04.02}}
 }
 Speech_C_2 {
  id=#9
  class=Speech_C
  outputs []
  properties {softTaskIcon=&T_UI_NPC_Intel2, generalConfig {defaultMessage=This will help us track traces of the comet...}}
  sub_nodes [
   ConversationChoiceNode_0 {
    class=ConversationChoiceNode
    properties {defaultChoiceDisplayText=Look into the Portal, choiceTags {gameplayTags [{tagName=Athena.Quests.S33.Conversation.Story.StopMM.04.02}]}}
   }
   FortConversationRequirement_ControllerRequirement_0 {
    class=FortConversationRequirement_ControllerRequirement
    properties {
     nodeName=Require: Is Quest Active
     requirement {
      class=FortControllerRequirement_IsQuestActive
      properties {questDefinition=&quest_S33_story_ally_q04}
     }
    }
   }
  ]
 }
 ConversationEntryPointNode_0 {
  id=#1
  class=ConversationEntryPointNode
  outputs [#10]
  properties {entryTag {tagName=Conversation.Entry.Season33.Story.StopMM.04.01}}
 }
 Speech_C_3 {
  id=#10
  class=Speech_C
  outputs []
  properties {
   softTaskIcon=&T_UI_NPC_Intel2
   generalConfig {defaultMessage=This will enhance the energy left behind by the comet.}
  }
  sub_nodes [
   ConversationChoiceNode_2 {
    class=ConversationChoiceNode
    properties {defaultChoiceDisplayText=Look into the Portal, choiceTags {gameplayTags [{tagName=Athena.Quests.S33.Conversation.Story.StopMM.04.01}]}}
   }
   FortConversationRequirement_ControllerRequirement_2 {
    class=FortConversationRequirement_ControllerRequirement
    properties {
     nodeName=Require: Is Quest Active
     requirement {
      class=FortControllerRequirement_IsQuestActive
      properties {questDefinition=&quest_S33_story_ally_q04}
     }
    }
   }
  ]
 }
]
