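# Season 33 story conversation database for StopMM quest 06, in which an NPC explains that the Island's comet is a shard of the Zero Point.
# A main-questline "The Spirit Realm" choice with a quest icon opens it, gated on quest_S33_story_ally_q06 being active.
# Four speech nodes run linearly through Continue choices, and the last choice carries the StopMM.06 quest gameplay tag.
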ConversationDatabase /BattlePassS33_Quests/Conversations/StoryDialogue/StopMM/Conversation_S33_StopMM_Q06
entry_points [{tag=Conversation.Entry.Season33.Story.StopMM.06, destinations [#1]}]
nodes [
 ConversationEntryPointNode_0 {
  id=#1
  class=ConversationEntryPointNode
  outputs [#2]
  properties {entryTag {tagName=Conversation.Entry.Season33.Story.StopMM.06}}
 }
 Speech_C_3 {
  id=#2
  class=Speech_C
  outputs [#3]
  properties {
   softTaskIcon=&T_UI_NPC_MainQuest
   generalConfig {defaultMessage=The comet that arrived on this Island... It's a piece of something greater.}
  }
  sub_nodes [
   ConversationChoiceNode_5 {
    class=ConversationChoiceNode
    properties {defaultChoiceDisplayText=The Spirit Realm, choiceTags {gameplayTags [{tagName=Conversation.Choice.Category.MainQuestline}]}}
   }
   FortConversationRequirement_ControllerRequirement_4 {
    class=FortConversationRequirement_ControllerRequirement
    properties {
     nodeName=Require: Is Quest Active
     requirement {
      class=FortControllerRequirement_IsQuestActive
      properties {questDefinition=/BattlePassS33_Quests/Items/QuestItems/Story/S33_Story_Ally/quest_S33_story_ally_q06}
     }
    }
   }
  ]
 }
 Speech_C_28 {
  id=#3
  class=Speech_C
  outputs [#4]
  properties {generalConfig {defaultMessage=A shard of something called the 'Zero Point'.}}
  sub_nodes [ConversationChoiceNode_6 {class=ConversationChoiceNode, properties {defaultChoiceDisplayText=Continue}}]
 }
 Speech_C_0 {
  id=#4
  class=Speech_C
  outputs [#5]
  properties {generalConfig {defaultMessage="I've never seen it, but Jones has. He said it's incredibly powerful and strange."}}
  sub_nodes [ConversationChoiceNode_7 {class=ConversationChoiceNode, properties {defaultChoiceDisplayText=Continue}}]
 }
 Speech_C_1 {
  id=#5
  class=Speech_C
  outputs []
  properties {generalConfig {defaultMessage=And if just a shard of it could cause all of this chaos... I believe him.}}
  sub_nodes [ConversationChoiceNode_8 {
    class=ConversationChoiceNode
    properties {defaultChoiceDisplayText=Continue, choiceTags {gameplayTags [{tagName=Athena.Quests.S33.Conversation.Story.StopMM.06}]}}
   }]
 }
]
